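# Niagara effects for the Bengal Basher glider cosmetic: headlights, back lights, twin exhaust, speed lines and four ribbon trails.
# Lights and trails follow skeletal mesh sockets or bones, with lights and exhaust scaled by time of day.
# System opacity and fade values drive intensity, while gameplay exhaust bends against owner velocity.
# FE_ emitters replace exhaust, speed lines and trails in the front end, taking direction from mesh orientation; a debug emitter is disabled.
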NiagaraSystem /BRCosmetics/Gadgets/Assets/VinderTech_GliderChute/Glider_BengalBasher/FX/NS_Glider_BengalBasher
dependencies {
 usedDataInterfaces [
  &NiagaraDataInterfaceVector2DCurve
  &NiagaraDataInterfaceCurve
  &NiagaraDataInterfaceColorCurve
  &NiagaraDataInterfaceArrayInt32
  &NiagaraDataInterfaceArrayFloat3
  &NiagaraDataInterfaceArrayFloat
 ]
 usedDynamicInputs [
  /Niagara/DynamicInputs/Math/Min_Float
  &ClampFloat
  &Subtract_Float
  &SetBoolByFloatComparison
  &Add_Float
  &Power
  &Abs_Float
 ]
 usedModules [
  &self:SystemSpawnScript.SystemScriptSource.SystemScriptGraph.NiagaraNodeAssignment_4.SetVariables_#1
  &SystemState
  &self:SystemSpawnScript.SystemScriptSource.SystemScriptGraph.NiagaraNodeAssignment_0.SetVariables_#2
  &self:SystemSpawnScript.SystemScriptSource.SystemScriptGraph.NiagaraNodeAssignment_5.SetVariables_#3
  &Timeline
  &self:SystemSpawnScript.SystemScriptSource.SystemScriptGraph.NiagaraNodeAssignment_1.SetVariables_#4
  &self:SystemSpawnScript.SystemScriptSource.SystemScriptGraph.NiagaraNodeAssignment_2.SetVariables_#5
  &self:SystemSpawnScript.SystemScriptSource.SystemScriptGraph.NiagaraNodeAssignment_3.SetVariables_#6
 ]
}
emitters {
 BackLight {
  emitterUpdateScript [
   &EmitterState(
    Inactive Response=::NewEnumerator1
    Life Cycle Mode=::NewEnumerator1
    Loop Behavior=::NewEnumerator1
    Loop Duration Mode=::NewEnumerator1
    UseLoopDelay=false
   )
   &SpawnBurst_Instantaneous(Spawn Count=2, Use Spawn Probability=false, UseLoopCountLimit=false) {module=SpawnBurst_Instantaneous001}
   &SpawnBurst_Instantaneous(Spawn Count=2, Spawn Group=1, Use Spawn Probability=false, UseLoopCountLimit=false)
  ]
  particleSpawnScript [
   /SetVariables(Particles.InitialSpawnGroup:NiagaraInt32=&SpawnGroup()) {isSetParameters=true}
   &FN_SetSpawnGroup()
   &InitializeParticle(
    Color=<rgba 0.9 0.0024 0.0137 0.4>
    Sprite Size=<33 33>
    Sprite Size Mode=::NewEnumerator4
    UsePositionOffset=true
    Write Material Random=true
   )
   &DynamicMaterialParameters(
    Index 0 Param 1=4
    Module.Index0Float0Write=true
    Module.Index0Float1Write=false
    Module.Index0Float2Write=false
    Module.Index0Float3Write=false
   )
  ]
  particleUpdateScript [
   &ParticleState(Kill Particles When Lifetime Has Elapsed=false)
   &SkeletalMeshLocation(
    Bone / Socket Index=&SelectIntFromArray(Direct Array Index=$Particles.InitialSpawnGroup, Int Selection Array=DataInterface() {propertyValues {IntData [8,9]}})
    MaskSpawnGroup=false
    Mesh Sampling Type=::NewEnumerator5
    Sampled Position Offset=&SelectVectorFromArray(
     Direct Array Index=$Particles.SpawnGroup
     Vector Selection Array=DataInterface() {propertyValues {FloatData [<0.5 -4 6>, <0 -4 -6>, <-0.5 -4 6>, <0 -4 -6>]}}
    )
    Skeletal Mesh=$System.Skeletal Mesh
    Socket Sampling Mode=::NewEnumerator5
   )
   /SetVariables(Particles.TimeOfDayHeadlightsIntensity:NiagaraFloat=&Multiply_Float(
     A=$System.HeadLightsFadeIn
     B=&FN_FrontEndFloatModifier(
      Gameplay=&FNTimeOfDayScalar(Night Scale=1, Morning Scale=0, Day Scale=0, Evening Scale=0)
      Front End=1
     )
    )) {isSetParameters=true}
   &CameraOffset()
   &ScaleSpriteSize(
    Scale Sprite Size Mode=::NewEnumerator0
    Uniform Scale Factor=&SelectFloatFromArray(Direct Array Index=$Particles.SpawnGroup, Float Selection Array=DataInterface() {propertyValues {FloatData [1.1,0.8,1.1,0.8]}})
   )
   &CameraOffset(Camera Offset Amount=10) {module=CameraOffset001}
   &ScaleColor(
    Scale Alpha=&Multiply_Float(
     A=&Multiply_Float(
      A=&FNTimeOfDayScalar(Night Scale=0.8, Morning Scale=0.9, Evening Scale=0.9)
      B=&FN_FrontEndFloatModifier(Gameplay=0.6, Front End=0.5)
     )
     B=$Particles.TimeOfDayHeadlightsIntensity
    )
    ScaleA=true
    ScaleRGB=false
   )
  ]
  rendererClasses [&NiagaraSpriteRendererProperties]
  renderers [{rendererClass=&NiagaraSpriteRendererProperties}]
 }
 Exhaust {
  emitterUpdateScript [
   &EmitterState(Life Cycle Mode=::NewEnumerator1, Loop Behavior=::NewEnumerator1, Loop Duration Mode=::NewEnumerator1, UseLoopDelay=false)
   &SpawnRate(
    Spawn Probability=&Multiply_Float(
     A=&Lerp_Float(A=0.8, B=1, Alpha=$System.Opacity)
     B=&FN_FrontEndFloatModifier(Gameplay=1, Front End=0)
    )
    SpawnRate=33
    Use Spawn Probability=true
   ) {module=SpawnRate002}
  ]
  particleSpawnScript [
   &FN_SetSpawnGroup(Number of Spawn Groups=2)
   /SetVariables(Particles.RandomFloat=&UniformAOrBFloat(A=-1)) {isSetParameters=true, module=SetVariables_#7}
   &InitializeParticle(
    Color=<rgba 0.6038 0.7529 1 0.025>
    Lifetime=&Lerp_Float(A=&UniformRangedFloat(Minimum=0.4, Maximum=0.8), B=0.8, Alpha=$System.Opacity)
    Position Offset=&SelectVectorFromArray(
     Direct Array Index=&SpawnGroup()
     Vector Selection Array=DataInterface() {propertyValues {FloatData [<-41.94 64 -4.4376>, <41.94 64 -4.4376>]}}
    )
    Sprite Rotation Angle=&UniformRangedFloat(Minimum=0, Maximum=360)
    Sprite Rotation Mode=::NewEnumerator0
    Sprite Size=&Vector2DFromFloat(Value=&Lerp_Float(
      A=&UniformRangedFloat(Minimum=40, Maximum=50)
      B=&UniformRangedFloat(Minimum=60, Maximum=80)
      Alpha=$System.Opacity
     ))
    Sprite Size Mode=::NewEnumerator4
    UsePositionOffset=true
    Write Material Random=true
   ) {module=InitializeParticle001}
  ]
  particleUpdateScript [
   &ParticleState()
   /SetVariables(Particles.Position=&Add_Vector(
     A=Linked() {linkedVariable {name=Particles.Position, type=NiagaraPosition}}
     B=&Multiply_VectorByFloat(
      Vector=&NormalizeVector(
       Vector To Normalize=&Lerp_Vector(
        A=<0 1 0>
        B=&TransformVector(
         Vector=&NormalizeVector(
          Vector To Normalize=&Multiply_Vector(A=$Engine.Owner.Velocity, B=<-1 -1 -3>)
          Safe Normalize=true
          Safe Fallback Threshold=0
          Safe Fallback Vector=<0 0 0.001>
         )
         Source Space=::World
         Destination Space=::Local
        )
        Alpha=&FloatFromCurve(
         FloatCurve=DataInterface() {propertyValues {Curve {keys [{interpMode=RCIM_Cubic, leaveTangent=2.5306, tangentMode=RCTM_Break, time=0, value=0}, {tangentMode=RCTM_User, time=1, value=1}]}}}
         Scale Curve=$System.ForwardAndRightDot
        )
       )
       Safe Normalize=true
       Safe Fallback Threshold=0
       Safe Fallback Vector=<0 0 0.001>
      )
      Float=&Lerp_Float(A=2, B=3, Alpha=$System.Opacity)
     )
    )) {isSetParameters=true}
   &ScaleColor(
    Scale Alpha=&FloatFromCurve(
     FloatCurve=DataInterface() {propertyValues {Curve {keys [
         <at 0 0 RCIM_Cubic>
         {arriveTangent=0.002, interpMode=RCIM_Cubic, leaveTangent=0.002, time=0.1559, value=1.0113}
         <at 1.0011 0.002 RCIM_Cubic>
        ]}}}
     Scale Curve=&Multiply_Float(
      A=$System.FadeIn
      B=&Multiply_Float(
       A=0.7
       B=&FNTimeOfDayScalar(Night Scale=0.35, Morning Scale=0.5, Day Scale=0.7, Evening Scale=0.5)
      )
     )
    )
    ScaleA=true
    ScaleRGB=true
   )
   /Niagara/Modules/Update/SubUV/SubUVAnimation(Number Of Frames=&FloatFromCurve(
     FloatCurve=DataInterface() {propertyValues {Curve {keys [
         {interpMode=RCIM_Cubic, leaveTangent=9.8981, tangentMode=RCTM_Break, time=0, value=0}
         {arriveTangent=1, interpMode=RCIM_Cubic, leaveTangent=1, tangentMode=RCTM_Break, time=0.075, value=0.5807}
         <at 1 1>
        ]}}}
     Scale Curve=&Multiply_Float(A=$Particles.RandomFloat, B=63)
    ))
   &ScaleSpriteSize(Scale Factor=&Vector2DFromFloat(Value=&FloatFromCurve(FloatCurve=DataInterface() {propertyValues {Curve {keys [{interpMode=RCIM_Cubic, tangentMode=RCTM_Break, time=0, value=0.179}, <at 0.8329 1>]}}})))
   &SpriteRotationRate(Rotation Rate=&Multiply_Float(A=$Particles.Initial.MaterialRandom, B=1))
  ]
  rendererClasses [&NiagaraSpriteRendererProperties]
  renderers [{rendererClass=&NiagaraSpriteRendererProperties}]
 }
 FE_Exhaust {
  emitterUpdateScript [
   &EmitterState(Life Cycle Mode=::NewEnumerator1, Loop Behavior=::NewEnumerator1, Loop Duration Mode=::NewEnumerator1, UseLoopDelay=false)
   &SpawnRate(
    Spawn Probability=&Multiply_Float(
     A=&Lerp_Float(A=0.8, B=1, Alpha=$System.Opacity)
     B=&FN_FrontEndFloatModifier(Gameplay=0, Front End=1)
    )
    SpawnRate=50
    Use Spawn Probability=true
   ) {module=SpawnRate002}
  ]
  particleSpawnScript [
   &FN_SetSpawnGroup(Number of Spawn Groups=2)
   /SetVariables(Particles.RandomFloat=&UniformAOrBFloat(A=-1)) {isSetParameters=true, module=SetVariables_#7}
   &InitializeParticle(
    Color=<rgba 0.6038 0.7529 1 0.025>
    Lifetime=&Lerp_Float(A=&UniformRangedFloat(Minimum=0.4, Maximum=0.8), B=0.8, Alpha=$System.Opacity)
    Position Offset=&SelectVectorFromArray(
     Direct Array Index=&SpawnGroup()
     Vector Selection Array=DataInterface() {propertyValues {FloatData [<-41.94 64 -4.4376>, <41.94 64 -4.4376>]}}
    )
    Sprite Rotation Angle=&UniformRangedFloat(Minimum=0, Maximum=360)
    Sprite Rotation Mode=::NewEnumerator0
    Sprite Size=&Vector2DFromFloat(Value=&Lerp_Float(
      A=&UniformRangedFloat(Minimum=40, Maximum=50)
      B=&UniformRangedFloat(Minimum=60, Maximum=80)
      Alpha=$System.Opacity
     ))
    Sprite Size Mode=::NewEnumerator4
    UsePositionOffset=true
    Write Material Random=true
   ) {module=InitializeParticle001}
   &SkeletalMeshLocation(
    Bone / Socket Index=&SelectIntFromArray(Direct Array Index=$Particles.SpawnGroup, Int Selection Array=DataInterface() {propertyValues {IntData [4,5]}})
    Bone Sampling Mode=::NewEnumerator4
    MaskSpawnGroup=false
    Orientation Sampling=::NewEnumerator1
    Skeletal Mesh=$System.Skeletal Mesh
   )
   &AddVelocity(
    Velocity=&ReturnMeshOrientationAxis(Mesh Orientation Axis=::XAxis, Mesh Orientation Quat=$Particles.SkeletalMeshLocation.SampledRotation)
    Velocity Speed Scale=&Multiply_Float(
     A=&Multiply_Float(A=100, B=&Lerp_Float(A=1, B=1.333, Alpha=$System.Opacity))
     B=&MaskFloatBySpawnGroup(Masked Float=1, Passthrough Float=-1)
    )
   )
  ]
  particleUpdateScript [
   &ParticleState()
   &ScaleColor(
    Scale Alpha=&FloatFromCurve(
     FloatCurve=DataInterface() {propertyValues {Curve {keys [
         <at 0 0 RCIM_Cubic>
         {arriveTangent=0.002, interpMode=RCIM_Cubic, leaveTangent=0.002, time=0.1559, value=1.0113}
         <at 1.0011 0.002 RCIM_Cubic>
        ]}}}
     Scale Curve=&Multiply_Float(
      A=&Lerp_Float(A=1, B=$System.FadeIn, Alpha=1)
      B=&Multiply_Float(
       A=0.7
       B=&FNTimeOfDayScalar(Night Scale=0.35, Morning Scale=0.5, Day Scale=0.7, Evening Scale=0.5)
      )
     )
    )
    ScaleA=true
    ScaleRGB=true
   )
   &SubUVAnimation(Number Of Frames=&FloatFromCurve(
     FloatCurve=DataInterface() {propertyValues {Curve {keys [
         {interpMode=RCIM_Cubic, leaveTangent=9.8981, tangentMode=RCTM_Break, time=0, value=0}
         {arriveTangent=1, interpMode=RCIM_Cubic, leaveTangent=1, tangentMode=RCTM_Break, time=0.075, value=0.5807}
         <at 1 1>
        ]}}}
     Scale Curve=&Multiply_Float(A=$Particles.RandomFloat, B=63)
    ))
   &ScaleSpriteSize(Scale Factor=&Vector2DFromFloat(Value=&FloatFromCurve(FloatCurve=DataInterface() {propertyValues {Curve {keys [{interpMode=RCIM_Cubic, tangentMode=RCTM_Break, time=0, value=0.179}, <at 0.8329 1>]}}})))
   &SolveForcesAndVelocity(Clamp Velocity=false, Limit Acceleration=false)
   &SpriteRotationRate(Rotation Rate=&Multiply_Float(A=$Particles.Initial.MaterialRandom, B=1))
  ]
  rendererClasses [&NiagaraSpriteRendererProperties]
  renderers [{rendererClass=&NiagaraSpriteRendererProperties}]
 }
 FE_Speedlines {
  emitterUpdateScript [
   &EmitterState()
   &SpawnRate(
    Spawn Probability=&FN_FrontEndFloatModifier(Gameplay=0, Front End=1)
    SpawnRate=&MakeCustomFloatFromBool(Bool=$User.bIsFrontEndPreview, True Float=40, False Float=90)
    Use Spawn Probability=true
   ) {module=SpawnRate001}
  ]
  particleSpawnScript [
   &InitializeParticle(
    Color=&MakeLinearColorFromVectorAndFloat(
     "Vector (RGB)"=<1 1 1>
     "Float (Alpha)"=&Multiply_Float(A=&UniformRangedFloat(Minimum=0.1, Maximum=0.18), B=0.4)
    )
    Lifetime Max=0.16
    Lifetime Min=0.1
    Lifetime Mode=::NewEnumerator1
    Position Offset=&FN_FrontEndVector3Modifier(Gameplay=<0 0 0>, Front End=<0 0 0>)
    Sprite Size Max=<3 10>
    Sprite Size Min=<2 2>
    Sprite Size Mode=::NewEnumerator2
    UsePositionOffset=true
    Write Material Random=true
   )
   &SampleSkeletalMesh(
    AdditionalYawPitchRoll=false
    Bone / Socket Index=6
    Bone Sampling Mode:ENiagaraBoneSamplingMode=::NewEnumerator6
    Skeletal Mesh:NiagaraDataInterfaceSkeletalMesh=$System.Skeletal Mesh
   )
   /SetVariables(Particles.CarOrientVector:Vector3f=&ReturnMeshOrientationAxis(Mesh Orientation Axis=::YAxis, Mesh Orientation Quat=$Particles.SampleSkeletalMesh.SampledRotation)) {isSetParameters=true}
   &TorusLocation(
    Align Torus With Axis=true
    Handle Radius=0
    Large Radius:NiagaraFloat=&Multiply_Float(A=&RandomRangeFloat(Minimum=60, Maximum=120), B=1.333)
    MaskSpawnGroup=false
    Module.Override Randomness=false
    Offset=&Add_Vector(A=&Multiply_VectorByFloat(Vector=$Particles.CarOrientVector, Float=-300), B=<0 0 30>)
    Torus Axis=<1 1 0>
   )
   &AddVelocityFromPoint(
    Module.Use Falloff Distance=false
    Origin Offset=&Add_Vector(A=&Multiply_VectorByFloat(Vector=$Particles.CarOrientVector, Float=-2000), B=<0 0 100>)
    Velocity Strength=&FN_FrontEndFloatModifier(Gameplay=600, Front End=600)
   ) {enabled=false}
   &AddVelocity(
    Velocity=&Multiply_VectorByFloat(Vector=$Particles.CarOrientVector, Float=&RandomRangeFloat(Minimum=300, Maximum=700))
    Velocity Speed Scale=&FN_FrontEndFloatModifier(Gameplay=1, Front End=1)
   )
  ]
  particleUpdateScript [
   &SolveForcesAndVelocity(Clamp Velocity=false, Limit Acceleration=false)
   &ParticleState()
   &ScaleSpriteSize(Scale Factor=&Vector2DFromCurve(
     Vector2Curve=DataInterface() {propertyValues {XCurve {keys [<at 0 0.5 RCIM_Cubic>,<at 0.5 1 RCIM_Cubic>]}, YCurve {keys [<at 0 0 RCIM_Cubic>,<at 0.5 1 RCIM_Cubic>]}}}
     ScaleCurve=&Vector2DFromFloat(Value=1)
    ))
   &ScaleSpriteSizeBySpeed(Max Scale Factor=<1 50>, Min Scale Factor=<0.01 0.05>, Velocity Threshold=1)
   &ScaleColor(
    Scale Alpha=&FloatFromCurve(
     FloatCurve=DataInterface() {propertyValues {Curve {keys [
         <at 0 0 RCIM_Cubic>
         <at 0.3759 1.0207>
         {arriveTangent=-2.221, interpMode=RCIM_Cubic, leaveTangent=-2.221, tangentMode=RCTM_User, time=1, value=0}
        ]}}}
     Scale Curve=&Lerp_Float(A=0, B=1, Alpha=$System.Opacity)
    )
    Scale RGB=&VectorFromFloat(Value=1)
    ScaleA=true
    ScaleRGB=true
   )
  ]
  rendererClasses [&NiagaraSpriteRendererProperties]
  renderers [{rendererClass=&NiagaraSpriteRendererProperties}]
 }
 FE_Trails {
  emitterSpawnScript [&FN_RibbonIDOffsetUtility()]
  emitterUpdateScript [
   &EmitterState()
   &FN_SpawnAnimTrail(Number of Anim Trails:NiagaraInt32=&Multiply_Int(
     A=4
     B=&MakeIntFromFloat(Float=&MakeCustomFloatFromBool(Bool=$User.bIsFrontEnd, True Float=1, False Float=0))
    ))
  ]
  particleSpawnScript [
   &FN_InitializeAnimTrailBoneIndecies(First Bone or Socket Index=1, Second Bone or Socket Index=0)
   &FN_InitializeAnimTrailBoneIndecies(AnimTrail Selection=1, First Bone or Socket Index=3, Second Bone or Socket Index=2) {module=FN_InitializeAnimTrailBoneIndecies001}
   &FN_InitializeAnimTrailBoneIndecies(AnimTrail Selection=2, First Bone or Socket Index=5, Second Bone or Socket Index=4) {module=FN_InitializeAnimTrailBoneIndecies002}
   &FN_InitializeAnimTrailBoneIndecies(AnimTrail Selection=3, First Bone or Socket Index=7, Second Bone or Socket Index=6) {module=FN_InitializeAnimTrailBoneIndecies003}
   /SetVariables(Particles.AlphaFadeIn:NiagaraFloat=$System.Opacity) {isSetParameters=true, module=SetVariables_#8}
   &SampleSkeletalMesh(
    AdditionalYawPitchRoll=false
    Bone / Socket Index=6
    Bone Sampling Mode:ENiagaraBoneSamplingMode=::NewEnumerator6
    Skeletal Mesh:NiagaraDataInterfaceSkeletalMesh=$System.Skeletal Mesh
   )
   &FN_InitializeAnimTrail(
    Color=&MultiplyLinearColor(
     Linear Color=<rgba 0.4497 0.4795 0.529 0.3>
     Scale Factor=&MakeFloat4FromVectorAndFloat("Vector (RGB)"=<1 1 1>, "Float (Alpha)"=$System.Opacity)
    )
    Lifetime=&MakeCustomFloatFromBool(Bool=$Particles.Locked, True Float=100, False Float=0.55)
    Position Offset=&FN_FrontEndVector3Modifier(
     Gameplay=<0 0 0>
     Front End=&Multiply_VectorByFloat(
      Vector=&ReturnMeshOrientationAxis(Mesh Orientation Axis=::YAxis, Mesh Orientation Quat=$Particles.SampleSkeletalMesh.SampledRotation)
      Float=-20
     )
    )
    SkeletalMesh=$System.Skeletal Mesh
   )
   &AddVelocity(Velocity=&Multiply_VectorByFloat(
     Vector=&Multiply_VectorByFloat(
      Vector=&ReturnMeshOrientationAxis(Mesh Orientation Axis=::YAxis, Mesh Orientation Quat=$Particles.SampleSkeletalMesh.SampledRotation)
      Float=800
     )
     Float=&MakeCustomFloatFromBool(Bool=$User.bIsFrontEnd, True Float=1, False Float=1)
    ))
   &DynamicMaterialParameters(
    Index 0 Param 1=1
    Module.Param0WriteEnabled=true
    Module.Param1WriteEnabled=true
    Module.Param2WriteEnabled=true
    Module.Param3WriteEnabled=true
   )
   &CameraOffset()
  ]
  particleUpdateScript [
   &SolveForcesAndVelocity(Clamp Velocity=false, Limit Acceleration=false)
   &FN_UpdateAnimTrail(
    Locked Particle Color:LinearColor=&MultiplyLinearColor(
     Linear Color=$Particles.Initial.Color
     Scale Factor=&MakeFloat4FromVectorAndFloat("Vector (RGB)"=<1 1 1>, "Float (Alpha)"=1)
    )
    Locked Particle RibbonWidth=1
    SkeletalMesh=$System.Skeletal Mesh
   ) {module=FN_UpdateAnimTrail001}
   /SetVariables(Particles.RibbonU0Override=$Particles.NormalizedAge) {isSetParameters=true, module=SetVariables_#9}
   &ScaleColor(
    Scale Alpha=&FNTimeOfDayScalar(Night Scale=1, Morning Scale=0.7, Evening Scale=0.7)
    ScaleA=true
    ScaleRGB=false
   ) {module=ScaleColor001}
  ]
  rendererClasses [&NiagaraRibbonRendererProperties]
  renderers [
   {rendererClass=&NiagaraRibbonRendererProperties}
   {rendererClass=&NiagaraRibbonRendererProperties, rendererIndex=1}
  ]
 }
 FN_Debug {
  bEnabled=false
  emitterUpdateScript [
   &EmitterState(Life Cycle Mode=::NewEnumerator1, Loop Behavior=::NewEnumerator1, Loop Duration Mode=::NewEnumerator1, UseLoopDelay=false)
   &SpawnBurst_Instantaneous(Spawn Count=1, Use Spawn Probability=false)
  ]
  particleSpawnScript [
   /Niagara/Modules/Spawn/Initialization/InitializeParticle(
    Lifetime=5
    Module.Write Color=true
    Module.Write Lifetime=true
    Module.Write Mass=true
    Module.Write Material Random=true
    Module.Write Position=false
    Module.Write Scale=false
    Module.Write SpriteRotation=false
    Module.Write SpriteSize=true
    Position=&SimulationPosition()
    Sprite Size=<100 100>
   )
   &SystemLocation(MaskSpawnGroup=false, Offset=<100 100 100>)
  ]
  particleUpdateScript [
   &ParticleState(Kill Particles When Lifetime Has Elapsed=false)
   &FN_Timer()
   /SetVariables(Particles.Debug:NiagaraFloat=$System.Opacity) {isSetParameters=true}
   &DynamicMaterialParameters(
    Index 0 Param 1=$Particles.Debug
    Module.Param0WriteEnabled=true
    Module.Param1WriteEnabled=true
    Module.Param2WriteEnabled=true
    Module.Param3WriteEnabled=true
   )
   &CameraOffset(Camera Offset Amount=111)
  ]
  rendererClasses [&NiagaraSpriteRendererProperties]
  renderers [{rendererClass=&NiagaraSpriteRendererProperties}]
 }
 Headlights {
  emitterUpdateScript [
   &EmitterState(
    Inactive Response=::NewEnumerator1
    Life Cycle Mode=::NewEnumerator1
    Loop Behavior=::NewEnumerator1
    Loop Duration Mode=::NewEnumerator1
    UseLoopDelay=false
   )
   &SpawnBurst_Instantaneous(
    Spawn Count=2
    Spawn Probability=&FN_FrontEndFloatModifier(Gameplay=1, Front End=0)
    Use Spawn Probability=true
    UseLoopCountLimit=false
   )
  ]
  particleSpawnScript [
   &FN_SetSpawnGroup(Number of Spawn Groups=2)
   /SetVariables(Particles.MirrorAxis:Vector3f=&MaskVectorBySpawnGroup(Masked Vector=<-1 1 1>, Passthrough Vector=<1 1 1>)) {isSetParameters=true}
   /Niagara/Modules/Spawn/Initialization/V2/InitializeParticle(
    Color=<rgba 0.7 0.73 1 1>
    Mesh Scale Mode=::NewEnumerator3
    Mesh Uniform Scale=5
    Position Offset=&Multiply_Vector(A=<79.3558 -330.8168 33>, B=$Particles.MirrorAxis)
    Sprite Size=<150 222>
    Sprite Size Mode=::NewEnumerator4
    UsePositionOffset=false
    Write Material Random=true
   )
   &DynamicMaterialParameters(
    Index 0 Param 1=&Multiply_Float(A=-19, B=&MakeFloatFromVector(Vector=$Particles.MirrorAxis, Channel=::NewEnumerator0))
    Index 0 Param 2=-1
    Module.Param0WriteEnabled=true
    Module.Param1WriteEnabled=true
    Module.Param2WriteEnabled=false
    Module.Param3WriteEnabled=false
   )
   /SetVariables(Particles.BeamColor:LinearColor=&MultiplyLinearColor(Linear Color=$Particles.Color, Scale Factor=&MakeVector4_FromFloat(Float=0.5))) {isSetParameters=true}
  ]
  particleUpdateScript [
   &SkeletalMeshLocation(
    Bone / Socket Index=&SelectIntFromArray(Direct Array Index=$Particles.SpawnGroup, Int Selection Array=DataInterface() {propertyValues {IntData [10,11]}})
    MaskSpawnGroup=false
    Mesh Sampling Type=::NewEnumerator5
    Orientation Sampling=::NewEnumerator1
    Skeletal Mesh=$User.SkeletalMesh
    Socket Sampling Mode=::NewEnumerator5
   )
   &SpriteFacingAndAlignment(
    Alignment Coordinate Space=::Local
    Sprite Alignment=&Add_Vector(
     A=&ReturnMeshOrientationAxis(Mesh Orientation Axis=::YAxis, Mesh Orientation Quat=$Particles.SkeletalMeshLocation.SampledRotation)
     B=&Multiply_Vector(A=<-0.05 0 0.1>, B=$Particles.MirrorAxis)
    )
    Sprite Facing=<0 1 0>
    WriteAlignment=true
    WriteFacing=false
   ) {module=SpriteFacingAndAlignment001}
   /SetVariables(
    Particles.FlareAlignment:Vector3f=&ReturnMeshOrientationAxis(Mesh Orientation Axis=::ZAxis, Mesh Orientation Quat=$Particles.SkeletalMeshLocation.SampledRotation)
    Particles.FlareCameraOffset=30
    Particles.FlareColor=&MultiplyLinearColor(Linear Color=$Particles.Color, Scale Factor=&MakeVector4_FromFloat(Float=1.2))
    Particles.FlareFacing:Vector3f=&Multiply_Vector(A=<0.25 -0.3792 0>, B=$Particles.MirrorAxis)
    Particles.FlarePosition:NiagaraPosition=&AddVectorToPosition(Position=$Particles.Position, Vector=&Multiply_Vector(A=<-1.8259 -10.9124 -0.8342>, B=$Particles.MirrorAxis))
    Particles.FlareSize:Vector2f=<100 50>
   ) {isSetParameters=true}
   /SetVariables(
    Particles.GlowColor=&MultiplyLinearColor(Linear Color=$Particles.Color, Scale Factor=&MakeVector4_FromFloat(Float=1))
    Particles.GlowPosition:NiagaraPosition=&AddVectorToPosition(Position=$Particles.Position, Vector=&Multiply_Vector(A=<-1 -7 -3>, B=$Particles.MirrorAxis))
    Particles.GlowSize=&Vector2DFromFloat(Value=30)
   ) {isSetParameters=true}
   /SetVariables(
    Particles.LightColor=&MultiplyLinearColor(Linear Color=$Particles.Color, Scale Factor=&MakeVector4_FromFloat(Float=11))
    Particles.LightPosition:NiagaraPosition=&AddVectorToPosition(Position=$Particles.Position, Vector=&Multiply_Vector(A=<5.4876 -23.0877 -1.6639>, B=$Particles.MirrorAxis))
   ) {isSetParameters=true}
   &ParticleState(Kill Particles When Lifetime Has Elapsed=false)
   /SetVariables(
    Particles.TimeOfDayHeadlightsIntensity:NiagaraFloat=&Multiply_Float(
     A=$System.HeadLightsFadeIn
     B=&FN_FrontEndFloatModifier(
      Gameplay=&FNTimeOfDayScalar(Night Scale=1, Morning Scale=0, Day Scale=0, Evening Scale=0)
      Front End=1
     )
    )
    Particles.TimeOfDayVisibilityTag:NiagaraInt32=&MakeIntFromFloat(Float=&FN_FrontEndFloatModifier(Gameplay=&FNTimeOfDayScalar(Night Scale=0, Morning Scale=1, Evening Scale=1), Front End=0))
   ) {isSetParameters=true}
   /SetVariables(
    Particles.BeamColor:LinearColor=&MultiplyLinearColor(
     Linear Color=$Particles.Initial.BeamColor
     Scale Factor=&MakeVector4_FromFloat(Float=$Particles.TimeOfDayHeadlightsIntensity)
    )
    Particles.Color=&MultiplyLinearColor(
     Linear Color=$Particles.Initial.Color
     Scale Factor=&MakeVector4_FromFloat(Float=$Particles.TimeOfDayHeadlightsIntensity)
    )
    Particles.FlareColor=&MultiplyLinearColor(
     Linear Color=$Particles.Initial.FlareColor
     Scale Factor=&MakeVector4_FromFloat(Float=$Particles.TimeOfDayHeadlightsIntensity)
    )
    Particles.GlowColor=&MultiplyLinearColor(
     Linear Color=$Particles.Initial.GlowColor
     Scale Factor=&MakeVector4_FromFloat(Float=$Particles.TimeOfDayHeadlightsIntensity)
    )
    Particles.LightColor=&MultiplyLinearColor(
     Linear Color=$Particles.Initial.LightColor
     Scale Factor=&MakeVector4_FromFloat(Float=$Particles.TimeOfDayHeadlightsIntensity)
    )
   ) {isSetParameters=true}
   &CameraOffset(Camera Offset Amount=11)
  ]
  rendererClasses [&NiagaraSpriteRendererProperties, &NiagaraLightRendererProperties]
  renderers [
   {rendererClass=&NiagaraSpriteRendererProperties}
   {rendererClass=&NiagaraLightRendererProperties, rendererIndex=1}
   {rendererClass=&NiagaraSpriteRendererProperties, rendererIndex=2}
   {rendererClass=&NiagaraSpriteRendererProperties, rendererIndex=3}
  ]
 }
 Speedlines {
  emitterUpdateScript [
   &EmitterState()
   &SpawnRate(
    Spawn Probability=&FN_FrontEndFloatModifier(Gameplay=1, Front End=0)
    SpawnRate=&MakeCustomFloatFromBool(Bool=$User.bIsFrontEndPreview, True Float=50, False Float=90)
    Use Spawn Probability=true
   ) {module=SpawnRate001}
  ]
  particleSpawnScript [
   &InitializeParticle(
    Color=&MakeLinearColorFromVectorAndFloat(
     "Vector (RGB)"=<1 1 1>
     "Float (Alpha)"=&Multiply_Float(
      A=&UniformRangedFloat(Minimum=0.1, Maximum=0.18)
      B=&MakeCustomFloatFromBool(Bool=$User.bIsFrontEndPreview, True Float=0.4, False Float=1)
     )
    )
    Lifetime Max=0.16
    Lifetime Min=0.1
    Lifetime Mode=::NewEnumerator1
    Position Offset=&FN_FrontEndVector3Modifier(Gameplay=<0 0 0>, Front End=<0 50 0>)
    Sprite Size Max=<3 10>
    Sprite Size Min=<2 2>
    Sprite Size Mode=::NewEnumerator2
    UsePositionOffset=true
    Write Material Random=true
   )
   &TorusLocation(
    Align Torus With Axis=true
    Handle Radius=0
    Large Radius:NiagaraFloat=&Multiply_Float(A=&RandomRangeFloat(Minimum=60, Maximum=180), B=1.5)
    MaskSpawnGroup=false
    Module.Override Randomness=false
    Offset=<0 -210 0>
    Torus Axis=<1 1 0>
   )
   &AddVelocityFromPoint(
    Module.Use Falloff Distance=false
    Origin Offset=<0 -550 0>
    Velocity Strength=&FN_FrontEndFloatModifier(Gameplay=600, Front End=600)
   ) {enabled=false}
   &AddVelocity(
    Velocity=&RandomRangeVector(Minimum=<0 300 0>, Maximum=<1 700 1>)
    Velocity Speed Scale=&FN_FrontEndFloatModifier(Gameplay=1, Front End=0)
   )
  ]
  particleUpdateScript [
   &SolveForcesAndVelocity(Clamp Velocity=false, Limit Acceleration=false)
   &ParticleState()
   &ScaleSpriteSize(Scale Factor=&Vector2DFromCurve(
     Vector2Curve=DataInterface() {propertyValues {XCurve {keys [<at 0 0.5 RCIM_Cubic>,<at 0.5 1 RCIM_Cubic>]}, YCurve {keys [<at 0 0 RCIM_Cubic>,<at 0.5 1 RCIM_Cubic>]}}}
     ScaleCurve=&Vector2DFromFloat(Value=1)
    ))
   &ScaleSpriteSizeBySpeed(Max Scale Factor=<1 50>, Min Scale Factor=<0.01 0.05>, Velocity Threshold=1)
   &ScaleColor(
    Scale Alpha=&FloatFromCurve(
     FloatCurve=DataInterface() {propertyValues {Curve {keys [
         <at 0 0 RCIM_Cubic>
         <at 0.3759 1.0207>
         {arriveTangent=-2.221, interpMode=RCIM_Cubic, leaveTangent=-2.221, tangentMode=RCTM_User, time=1, value=0}
        ]}}}
     Scale Curve=&MakeCustomFloatFromBool(
      Bool=$User.bIsFrontEndPreview
      True Float=&MakeCustomFloatFromBool(Bool=$User.bIsFullyDeployed, True Float=1, False Float=0)
      False Float=&Lerp_Float(A=0, B=0.9, Alpha=$System.Opacity)
     )
    )
    Scale RGB=&VectorFromFloat(Value=1)
    ScaleA=true
    ScaleRGB=true
   )
  ]
  rendererClasses [&NiagaraSpriteRendererProperties]
  renderers [{rendererClass=&NiagaraSpriteRendererProperties}]
 }
 Trails {
  emitterSpawnScript [&FN_RibbonIDOffsetUtility()]
  emitterUpdateScript [
   &EmitterState()
   &FN_SpawnAnimTrail(Number of Anim Trails:NiagaraInt32=&Multiply_Int(
     A=4
     B=&MakeIntFromFloat(Float=&MakeCustomFloatFromBool(Bool=$User.bIsFrontEnd, True Float=0, False Float=1))
    ))
  ]
  particleSpawnScript [
   &FN_InitializeAnimTrailBoneIndecies(First Bone or Socket Index=1, Second Bone or Socket Index=0)
   &FN_InitializeAnimTrailBoneIndecies(AnimTrail Selection=1, First Bone or Socket Index=3, Second Bone or Socket Index=2) {module=FN_InitializeAnimTrailBoneIndecies001}
   &FN_InitializeAnimTrailBoneIndecies(AnimTrail Selection=2, First Bone or Socket Index=5, Second Bone or Socket Index=4) {module=FN_InitializeAnimTrailBoneIndecies002}
   &FN_InitializeAnimTrailBoneIndecies(AnimTrail Selection=3, First Bone or Socket Index=7, Second Bone or Socket Index=6) {module=FN_InitializeAnimTrailBoneIndecies003}
   &FN_SetSpawnGroup()
   /SetVariables(Particles.AlphaFadeIn:NiagaraFloat=$System.Opacity) {isSetParameters=true, module=SetVariables_#8}
   &FN_InitializeAnimTrail(
    AnimTrailWidth=&FN_FrontEndFloatModifier(Gameplay=1, Front End=0.7)
    Color=&MultiplyLinearColor(
     Linear Color=<rgba 0.4497 0.4795 0.529 0.3>
     Scale Factor=&MakeFloat4FromVectorAndFloat("Vector (RGB)"=<1 1 1>, "Float (Alpha)"=$System.Opacity)
    )
    Lifetime=&MakeCustomFloatFromBool(Bool=$Particles.Locked, True Float=100, False Float=0.55)
    Position Offset=&FN_FrontEndVector3Modifier(
     Gameplay=&MakeVector(
      X=0
      Y=&SelectFloatFromArray(Direct Array Index=$Particles.SpawnGroup, Float Selection Array=DataInterface() {propertyValues {FloatData [-20,-20,0,0]}})
      Z=0
     )
     Front End=<0 0 0>
    )
    SkeletalMesh=$System.Skeletal Mesh
   )
   &AddVelocity(Velocity=&Multiply_VectorByFloat(
     Vector=&Lerp_Vector(A=<0 800 0>, B=<0 800 0>, Alpha=&FN_FrontEndFloatModifier(Gameplay=0, Front End=1))
     Float=&MakeCustomFloatFromBool(Bool=$User.bIsFrontEnd, True Float=1, False Float=1)
    ))
   &DynamicMaterialParameters(
    Index 0 Param 1=1
    Module.Param0WriteEnabled=true
    Module.Param1WriteEnabled=true
    Module.Param2WriteEnabled=true
    Module.Param3WriteEnabled=true
   )
   &CameraOffset()
  ]
  particleUpdateScript [
   &SolveForcesAndVelocity(Clamp Velocity=false, Limit Acceleration=false)
   &FN_UpdateAnimTrail(
    Locked Particle Color:LinearColor=&MultiplyLinearColor(
     Linear Color=$Particles.Initial.Color
     Scale Factor=&MakeFloat4FromVectorAndFloat("Vector (RGB)"=<1 1 1>, "Float (Alpha)"=1)
    )
    Locked Particle RibbonWidth=1
    SkeletalMesh=$System.Skeletal Mesh
   ) {module=FN_UpdateAnimTrail001}
   /SetVariables(Particles.RibbonU0Override=$Particles.NormalizedAge) {isSetParameters=true, module=SetVariables_#9}
   &ScaleColor(
    Scale Alpha=&FNTimeOfDayScalar(Night Scale=1, Morning Scale=0.7, Evening Scale=0.7)
    ScaleA=true
    ScaleRGB=false
   ) {module=ScaleColor001}
  ]
  rendererClasses [&NiagaraRibbonRendererProperties]
  renderers [
   {rendererClass=&NiagaraRibbonRendererProperties}
   {rendererClass=&NiagaraRibbonRendererProperties, rendererIndex=1}
  ]
 }
}
summary {
 systemName=NS_Glider_BengalBasher
 emitters [
  {bEnabled=false, emitterName=FN_Debug, rendererClasses [&NiagaraSpriteRendererProperties]}
  {emitterName=Speedlines, rendererClasses [&NiagaraSpriteRendererProperties]}
  {emitterName=Headlights, rendererClasses [&NiagaraSpriteRendererProperties, &NiagaraLightRendererProperties]}
  {emitterName=Exhaust, rendererClasses [&NiagaraSpriteRendererProperties]}
  {emitterName=Trails, rendererClasses [&NiagaraRibbonRendererProperties]}
  {emitterName=BackLight, rendererClasses [&NiagaraSpriteRendererProperties]}
  {emitterName=FE_Trails, rendererClasses [&NiagaraRibbonRendererProperties]}
  {emitterName=FE_Speedlines, rendererClasses [&NiagaraSpriteRendererProperties]}
  {emitterName=FE_Exhaust, rendererClasses [&NiagaraSpriteRendererProperties]}
 ]
 userVariables [
  User.bIsFrontEnd {type=NiagaraBool}
  User.bIsFrontEndPreview {type=NiagaraBool}
  User.bIsFullyDeployed {type=NiagaraBool}
  User.ForwardDot {type=NiagaraFloat, defaultValue=1, description=""}
  User.RibbonIDOffset {type=NiagaraInt32}
  User.RightDot {type=NiagaraFloat, defaultValue=0, description=""}
  User.SkeletalMesh {
   type=NiagaraDataInterfaceSkeletalMesh
   defaultValue {
    struct=NiagaraExt_VariableValue_DataInterface
    value {dataInterface=&self:NiagaraDataInterfaceSkeletalMesh_0}
   }
  }
 ]
}
